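# Sound cue for the Nurse Husk's healing loop, a looping enemy sound in the Fort monster audio set.
# A looping Wave Player feeds a Modulator left at default settings, which is the root of the cue.
# The cue plays slightly below full volume in the enemy attack sound class, borrows the gadget-activate attenuation asset, and overrides concurrency.
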SoundCue /Game/Sounds/Fort_Monster_Husk_Nurse/SFX_Nurse_Husk_Healing_Loop_Cue
root=Modulator_2
nodes [
 {id=Modulator_2, children [WavePlayer_0]}
 {id=WavePlayer_0, properties {bLooping=true, soundWaveAssetPtr=&/SFX_Nurse_Husk_Healing_Loop}}
]
properties {
 attenuationSettings=&Fort_GadgetActivate_Att
 bOverrideConcurrency=true
 maxDistance=4000
 soundClassObject=/Game/Sounds/FortSoundClassesAndModes/Fort_Enemy_Attack
 volumeMultiplier=0.9
}
